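# Physics asset for the F_MED_Lurk female body that gives secondary motion to jacket, collar, lapels, hem zipper and laces.
# Capsules on pelvis, spine and calf twist bones anchor the dynamic bones, with free constraints linking them back to the pelvis.
# Jacket, collar and lapel bodies hang from spine_04 with limited swing and mostly tight twist, and the zipper chains off the left jacket body.
# Lace bodies hang from their calf with unrestricted rotation.
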PhysicsAsset /BRCosmetics/Characters/Player/Female/Medium/Bodies/F_MED_Lurk/Meshes/F_MED_Lurk_Physics
bodies {
 calf_twist_01_l [{
   shapeType=Capsule
   length=5.6681
   radius=5.0102
   shapeName=calf_twist_01_l_capsule
   center=<-0.4835 0.6755 -0.0841>
   rotation=<rot 0 -13.8033 -87.4345>
  }]
 calf_twist_01_r [{
   shapeType=Capsule
   length=5.6082
   radius=5.0069
   shapeName=calf_twist_01_r_capsule
   center=<0.603 -0.7186 0.0872>
   rotation=<rot 0 -12.1727 -87.6732>
  }]
 dyn_B_jacket_dyn2_jnt [{shapeType=Capsule, length=4, radius=4, shapeName=dyn_B_jacket_dyn2_jnt_capsule, center=<0 -2.0555 0>}]
 dyn_LB_jacket_dyn2_jnt [{shapeType=Capsule, length=4, radius=4, shapeName=dyn_LB_jacket_dyn2_jnt_capsule, center=<2.0265 0 0.0001>, rotation=<rot 0.0001 0 -69.9999>}]
 dyn_L_L_lace_jnt [{
   shapeType=Capsule
   length=5.107
   radius=1.1292
   shapeName=dyn_L_L_lace_jnt_capsule
   center=<0.142 -0.5256 4.6058>
   rotation=<rot -4.8594 -90 4.5114>
  }]
 dyn_L_R_lace_jnt [{
   shapeType=Capsule
   length=3.9586
   radius=0.9465
   shapeName=dyn_L_R_lace_jnt_capsule
   center=<0.2798 -0.7228 3.5078>
   rotation=<rot -9.3633 -90 3.1817>
  }]
 dyn_L_collar_flap_jnt [{
   shapeType=Capsule
   length=5
   radius=3
   shapeName=dyn_L_collar_flap_jnt_capsule
   center=<0 -0.0908 -0.2493>
   rotation=<rot -0.0001 0.0001 109.9996>
  }]
 dyn_L_hem_zipper_jnt [{
   shapeType=Capsule
   length=1.255
   radius=0.755
   shapeName=dyn_L_hem_zipper_jnt_capsule
   center=<0.9568 -0.0039 0.0047>
   rotation=<rot 0 89.6313 -89.944>
  }]
 dyn_L_jacket_dyn2_jnt [{
   shapeType=Capsule
   length=6.75
   radius=4.75
   shapeName=dyn_L_jacket_dyn2_jnt_capsule
   center=<4.1284 -0.5139 -3.126>
   rotation=<rot -90 0 -79.9994>
  }]
 dyn_L_lapel_jnt [{shapeType=Capsule, length=4, radius=2.25, shapeName=dyn_L_lapel_jnt_capsule, center=<0.299 1.0664 2.1926>, rotation=<rot -90 -5.0508 5.0504>}]
 dyn_RB_jacket_dyn2_jnt [{
   shapeType=Capsule
   length=4
   radius=4
   shapeName=dyn_RB_jacket_dyn2_jnt_capsule
   center=<-1.7525 1.3785 -0.5017>
   rotation=<rot -3.4048 9.4096 -70.2801>
  }]
 dyn_R_L_lace_jnt [{
   shapeType=Capsule
   length=3.8139
   radius=0.7564
   shapeName=dyn_R_L_lace_jnt_capsule
   center=<-0.2812 0.0398 -3.9237>
   rotation=<rot 0.2781 -90 2.8791>
  }]
 dyn_R_R_lace_jnt [{
   shapeType=Capsule
   length=3.7741
   radius=1.1088
   shapeName=dyn_R_R_lace_jnt_capsule
   center=<-0.7224 0.4186 -3.2663>
   rotation=<rot -7.5013 -90 15.894>
  }]
 dyn_R_collar_flap_jnt [{shapeType=Capsule, length=5.25, radius=2.75, shapeName=dyn_R_collar_flap_jnt_capsule, rotation=<rot 0 0.0001 89.9995>}]
 dyn_R_jacket_dyn2_jnt [{
   shapeType=Capsule
   length=11.6475
   radius=4.1229
   shapeName=dyn_R_jacket_dyn2_jnt_capsule
   center=<4.6368 -3.2448 -0.5026>
   rotation=<rot 0 -165.4476 -83.3409>
  }]
 dyn_R_lapel_jnt [{shapeType=Capsule, length=4, radius=2.25, shapeName=dyn_R_lapel_jnt_capsule, center=<0 -1.2797 -1.1454>, rotation=<rot 90 0.0199 0.0202>}]
 pelvis [{shapeType=Capsule, length=7, radius=12, shapeName=pelvis_capsule, center=<1.4836 -1.514 0>}]
 spine_01 [{shapeType=Capsule, length=0.505, radius=0.505, shapeName=spine_01_capsule, rotation=<rot 0 -90 0>}]
 spine_02 [{shapeType=Capsule, length=6.0304, radius=6.9458, shapeName=spine_02_capsule, center=<0 1.5985 0>}]
 spine_04 [
  {shapeType=Capsule, length=2, radius=2, shapeName=spine_04_capsule}
  {shapeType=Capsule, length=15.75, radius=7.5, shapeName=spine_04_capsule, center=<7.3848 0.5715 0>}
 ]
}
constraints [
 {bone1Name=dyn_L_collar_flap_jnt, bone2Name=spine_04, swing1LimitDegrees=45, swing2LimitDegrees=45, twistLimitDegrees=45}
 {bone1Name=dyn_R_collar_flap_jnt, bone2Name=spine_04, swing1LimitDegrees=45, swing2LimitDegrees=45, twistLimitDegrees=45}
 {bone1Name=dyn_L_lapel_jnt, bone2Name=spine_04, swing1LimitDegrees=15, swing2LimitDegrees=15, twistLimitDegrees=1}
 {bone1Name=dyn_R_lapel_jnt, bone2Name=spine_04, swing1LimitDegrees=15, swing2LimitDegrees=15, twistLimitDegrees=1}
 {bone1Name=dyn_R_jacket_dyn2_jnt, bone2Name=spine_04, swing1LimitDegrees=65, swing2LimitDegrees=65, twistLimitDegrees=1}
 {bone1Name=dyn_B_jacket_dyn2_jnt, bone2Name=spine_04, swing1LimitDegrees=45, swing2LimitDegrees=10, twistLimitDegrees=2}
 {bone1Name=dyn_L_jacket_dyn2_jnt, bone2Name=spine_04, swing1LimitDegrees=55, swing2LimitDegrees=55, twistLimitDegrees=2}
 {bone1Name=dyn_LB_jacket_dyn2_jnt, bone2Name=spine_04, swing1LimitDegrees=5, swing2LimitDegrees=45, twistLimitDegrees=2}
 {bone1Name=dyn_RB_jacket_dyn2_jnt, bone2Name=spine_04, swing1LimitDegrees=5, swing2LimitDegrees=45, twistLimitDegrees=2}
 {
  bone1Name=dyn_R_L_lace_jnt
  bone2Name=calf_twist_01_r
  swing1LimitDegrees=45
  swing1Motion=Free
  swing2LimitDegrees=45
  swing2Motion=Free
  twistLimitDegrees=45
  twistMotion=Free
 }
 {
  bone1Name=dyn_R_R_lace_jnt
  bone2Name=calf_twist_01_r
  swing1LimitDegrees=45
  swing1Motion=Free
  swing2LimitDegrees=45
  swing2Motion=Free
  twistLimitDegrees=45
  twistMotion=Free
 }
 {
  bone1Name=dyn_L_L_lace_jnt
  bone2Name=calf_twist_01_l
  swing1LimitDegrees=45
  swing1Motion=Free
  swing2LimitDegrees=45
  swing2Motion=Free
  twistLimitDegrees=45
  twistMotion=Free
 }
 {
  bone1Name=dyn_L_R_lace_jnt
  bone2Name=calf_twist_01_l
  swing1LimitDegrees=45
  swing1Motion=Free
  swing2LimitDegrees=45
  swing2Motion=Free
  twistLimitDegrees=45
  twistMotion=Free
 }
 {
  bone1Name=spine_04
  bone2Name=spine_02
  swing1LimitDegrees=45
  swing1Motion=Free
  swing2LimitDegrees=45
  swing2Motion=Free
  twistLimitDegrees=45
  twistMotion=Free
 }
 {
  bone1Name=spine_02
  bone2Name=spine_01
  swing1LimitDegrees=45
  swing1Motion=Free
  swing2LimitDegrees=45
  swing2Motion=Free
  twistLimitDegrees=45
  twistMotion=Free
 }
 {
  bone1Name=calf_twist_01_l
  bone2Name=pelvis
  swing1LimitDegrees=45
  swing1Motion=Free
  swing2LimitDegrees=45
  swing2Motion=Free
  twistLimitDegrees=45
  twistMotion=Free
 }
 {
  bone1Name=calf_twist_01_r
  bone2Name=pelvis
  swing1LimitDegrees=45
  swing1Motion=Free
  swing2LimitDegrees=45
  swing2Motion=Free
  twistLimitDegrees=45
  twistMotion=Free
 }
 {
  bone1Name=spine_01
  bone2Name=pelvis
  swing1LimitDegrees=45
  swing1Motion=Free
  swing2LimitDegrees=45
  swing2Motion=Free
  twistLimitDegrees=45
  twistMotion=Free
 }
 {bone1Name=dyn_L_hem_zipper_jnt, bone2Name=dyn_L_jacket_dyn2_jnt, swing1LimitDegrees=15, swing2LimitDegrees=15, twistLimitDegrees=0}
]
