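# Physics asset for the zippers on the F_MED_Lurk female medium player body cosmetic, covering four dynamic zipper joints (upper and lower, left and right).
# Each zipper joint has a small capsule body, and a larger capsule on spine_04 acts as their shared anchor.
# Each zipper is constrained directly to spine_04 with narrow swing limits and near-zero twist, allowing only slight dangling motion.
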PhysicsAsset /BRCosmetics/Characters/Player/Female/Medium/Bodies/F_MED_Lurk/Meshes/F_MED_Lurk_PhysicsZippers
bodies {
 dyn_L_lower_zipper_jnt [{
   shapeType=Capsule
   length=0.7
   radius=0.7
   shapeName=dyn_L_lower_zipper_jnt_capsule
   center=<0.5737 0 -0.1012>
   rotation=<rot -90 -1.8984 1.8984>
  }]
 dyn_L_upper_zipper_jnt [{shapeType=Capsule, length=0.7, radius=0.7, shapeName=dyn_L_upper_zipper_jnt_capsule, center=<0.9807 0 0>, rotation=<rot -90 1.2658 -1.2658>}]
 dyn_R_lower_zipper_jnt [{shapeType=Capsule, length=0.7, radius=0.7, shapeName=dyn_R_lower_zipper_jnt_capsule, center=<0.9516 -0.0001 -0.1678>, rotation=<rot 80 0 0>}]
 dyn_R_upper_zipper_jnt [{
   shapeType=Capsule
   length=0.7
   radius=0.7
   shapeName=dyn_R_upper_zipper_jnt_capsule
   center=<1.0916 0.0001 0>
   rotation=<rot 90 0.6328 0.6326>
  }]
 spine_04 [{shapeType=Capsule, length=5, radius=5, shapeName=spine_04_capsule}]
}
constraints [
 {bone1Name=dyn_L_lower_zipper_jnt, bone2Name=spine_04, swing1LimitDegrees=15, swing2LimitDegrees=15, twistLimitDegrees=1}
 {bone1Name=dyn_L_upper_zipper_jnt, bone2Name=spine_04, swing1LimitDegrees=15, swing2LimitDegrees=15, twistLimitDegrees=1}
 {bone1Name=dyn_R_upper_zipper_jnt, bone2Name=spine_04, swing1LimitDegrees=15, swing2LimitDegrees=15, twistLimitDegrees=1}
 {bone1Name=dyn_R_lower_zipper_jnt, bone2Name=spine_04, swing1LimitDegrees=15, swing2LimitDegrees=15, twistLimitDegrees=0.2}
]
