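# Animation Blueprint for the F_MED_Lurk female player body cosmetic, following the attached parent mesh and adding deformation and jacket secondary motion.
# The copied pose runs through twist and sculpter Control Rigs, then two RigidBody simulations using body and zipper physics assets.
# Additive bone offsets ease both collar flaps into an adjusted pose while bIsTargeting is true.
# Weighted multi-bone Constraint nodes, resembling skin weights, finally position five zipper joints.
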AnimBlueprint /BRCosmetics/Characters/Player/Female/Medium/Bodies/F_MED_Lurk/Meshes/F_MED_Lurk_AnimBP
target_skeleton=&/F_MED_Lurk_Skeleton
graphs [
 AnimGraph {nodes [
   n1 = Control Rig(Source=Copy Pose From Mesh() {bUseAttachedParent=true}) {
    alphaBoolBlend {blendOption=Linear}
    controlRigAssetReference {blueprintRigClass=&/F_MED_Lurk_Deform_CtrlRig_twists.F_MED_Lurk_Deform_CtrlRig_twists_C}
   }
   n2 = Control Rig!(
    Source=@n1
    Alpha
    KneeRot_fr_l_114deg:Vector
    KneeTrans_fr_l_114deg:Vector="0.258,-3.591,0.009"
    KneeScale_fr_l_114deg:Vector="1,1,1"
    KneeScale_fr_l_130deg:Vector="1,1,1"
    KneeRot_fr_l_130deg:Vector="0,0,-51.2"
    KneeTrans_fr_l_130deg:Vector="0.258,-3.591,0.009"
   ) -> (Pose) {
    alphaBoolBlend {blendOption=Linear}
    controlRigAssetReference {blueprintRigClass=&/F_MED_Lurk_Deform_CtrlRig_sculpters.F_MED_Lurk_Deform_CtrlRig_sculpters_C}
   }
   n3 = RigidBody(ComponentPose=Local To Component(LocalPose=@n2)) {
    bClampLinearTranslationLimitToRefPose=true
    bOverrideWorldGravity=true
    overridePhysicsAsset=&/F_MED_Lurk_Physics
    simulationSpace=BaseBoneSpace
    alphaBoolBlend {blendOption=Linear}
    baseBoneRef {boneName=spine_01}
    componentAppliedLinearAccClamp=<3000 3000 3000>
    componentLinearAccScale=<0.4 0.3 0.3>
    componentLinearVelScale=<0.4 0.3 0.3>
    overrideWorldGravity {z=-980}
   }
   n4 = RigidBody(ComponentPose=@n3) {
    bClampLinearTranslationLimitToRefPose=true
    bOverrideWorldGravity=true
    overridePhysicsAsset=&/F_MED_Lurk_PhysicsZippers
    simulationSpace=BaseBoneSpace
    alphaBoolBlend {blendOption=Linear}
    baseBoneRef {boneName=spine_05}
    componentAppliedLinearAccClamp=<3000 3000 3000>
    componentLinearAccScale=<0.3 0.2 0.2>
    componentLinearVelScale=<0.3 0.2 0.2>
    overrideWorldGravity {z=-980}
   }
   n5 = "To Float (Boolean)"(InBool=Get bIsTargeting())
   n6 = "Transform (Modify) Bone - Bone: dyn_L_collar_flap_jnt"(ComponentPose=@n4, Alpha=@n5, Translation="0,0,2.3", Rotation="-50,20,0", Scale="1,1,1") {
    rotationMode=BMM_Additive
    rotationSpace=BCS_BoneSpace
    translationMode=BMM_Additive
    translationSpace=BCS_BoneSpace
    alphaBoolBlend {blendOption=Linear}
    alphaScaleBiasClamp {bInterpResult=true, interpSpeedIncreasing=5}
    boneToModify {boneName=dyn_L_collar_flap_jnt}
    rotation {pitch=-50, yaw=20}
    translation {z=2.3}
   }
   n7 = "Transform (Modify) Bone - Bone: dyn_R_collar_flap_jnt"(ComponentPose=@n6, Alpha=@n5, Translation="0,0,-2.3", Rotation="-30,0,0", Scale="1,1,1") {
    rotationMode=BMM_Additive
    rotationSpace=BCS_BoneSpace
    translationMode=BMM_Additive
    translationSpace=BCS_BoneSpace
    alphaBoolBlend {blendOption=Linear}
    alphaScaleBiasClamp {bInterpResult=true, interpSpeedIncreasing=5}
    boneToModify {boneName=dyn_R_collar_flap_jnt}
    rotation {pitch=-30}
    translation {z=-2.3}
   }
   n8 = "Constraint - L_upper_zipper_jnt (PPPPPPPPPPPPPPPPPPPPP)"!(
    ComponentPose=@n7
    Alpha=1
    ConstraintWeights_0=0
    ConstraintWeights_1:float=0.0001
    ConstraintWeights_2:float=0.0003
    ConstraintWeights_3:float=0.001
    ConstraintWeights_4:float=0.4796
    ConstraintWeights_5:float=0.1658
    ConstraintWeights_6:float=0.116
    ConstraintWeights_7:float=0.0911
    ConstraintWeights_8:float=0.0235
    ConstraintWeights_9:float=0
    ConstraintWeights_10:float=0.0833
    ConstraintWeights_11:float=0
    ConstraintWeights_12:float=0.0003
    ConstraintWeights_13:float=0.0002
    ConstraintWeights_14:float=0.0001
    ConstraintWeights_15:float=0
    ConstraintWeights_16:float=0.0014
    ConstraintWeights_17:float=0.0372
    ConstraintWeights_18:float=0
    ConstraintWeights_19:float=0
    ConstraintWeights_20:float=0.0001
   ) -> (Pose) {
    alphaBoolBlend {blendOption=Linear}
    boneToModify {boneName=L_upper_zipper_jnt}
    constraintSetup [
     {targetBone {boneName=pelvis}}
     {targetBone {boneName=spine_01}}
     {targetBone {boneName=spine_02}}
     {targetBone {boneName=spine_03}}
     {targetBone {boneName=spine_04}}
     {targetBone {boneName=spine_05}}
     {targetBone {boneName=clavicle_l}}
     {targetBone {boneName=upperarm_twist_01_l}}
     {targetBone {boneName=deform_deltoid_fr_l}}
     {targetBone {boneName=upperarm_twist_02_l}}
     {targetBone {boneName=deform_pectoral_fr_l}}
     {targetBone {boneName=clavicle_r}}
     {targetBone {boneName=neck_01}}
     {targetBone {boneName=neck_02}}
     {targetBone {boneName=head}}
     {targetBone {boneName=C_jaw}}
     {targetBone {boneName=dyn_L_collar_flap_jnt}}
     {targetBone {boneName=dyn_L_lapel_jnt}}
     {targetBone {boneName=dyn_LB_jacket_dyn2_jnt}}
     {targetBone {boneName=dyn_L_jacket_dyn1_jnt}}
     {targetBone {boneName=dyn_L_jacket_dyn2_jnt}}
    ]
    constraintWeights [1,1,1,1,1,1,1,1,1,1,1,1,1,1,1,1,1,1,1,1,1]
   }
   n9 = "Constraint - R_lower_zipper_jnt (PPPPPPPPPPPPPPPPPPPPPPP)"!(
    ComponentPose=@n8
    Alpha=1
    ConstraintWeights_0=0.0034
    ConstraintWeights_1:float=0.0096
    ConstraintWeights_2:float=0.0293
    ConstraintWeights_3:float=0.0569
    ConstraintWeights_4:float=0.6532
    ConstraintWeights_5:float=0.0651
    ConstraintWeights_6:float=0.0004
    ConstraintWeights_7:float=0
    ConstraintWeights_8:float=0.071
    ConstraintWeights_9:float=0.0693
    ConstraintWeights_10:float=0.0038
    ConstraintWeights_11:float=0.0001
    ConstraintWeights_12:float=0.007
    ConstraintWeights_13:float=0
    ConstraintWeights_14:float=0
    ConstraintWeights_15:float=0
    ConstraintWeights_16:float=0.0016
    ConstraintWeights_17:float=0.0057
    ConstraintWeights_18:float=0.0146
    ConstraintWeights_19:float=0
    ConstraintWeights_20:float=0.007
    ConstraintWeights_21:float=0
    ConstraintWeights_22:float=0.0018
   ) -> (Pose) {
    alphaBoolBlend {blendOption=Linear}
    boneToModify {boneName=R_lower_zipper_jnt}
    constraintSetup [
     {targetBone {boneName=pelvis}}
     {targetBone {boneName=spine_01}}
     {targetBone {boneName=spine_02}}
     {targetBone {boneName=spine_03}}
     {targetBone {boneName=spine_04}}
     {targetBone {boneName=spine_05}}
     {targetBone {boneName=clavicle_l}}
     {targetBone {boneName=upperarm_twist_02_l}}
     {targetBone {boneName=clavicle_r}}
     {targetBone {boneName=upperarm_twist_01_r}}
     {targetBone {boneName=deform_deltoid_fr_r}}
     {targetBone {boneName=upperarm_twist_02_r}}
     {targetBone {boneName=deform_pectoral_fr_r}}
     {targetBone {boneName=neck_01}}
     {targetBone {boneName=neck_02}}
     {targetBone {boneName=dyn_R_collar_flap_jnt}}
     {targetBone {boneName=dyn_R_lapel_jnt}}
     {targetBone {boneName=dyn_R_jacket_dyn1_jnt}}
     {targetBone {boneName=dyn_R_jacket_dyn2_jnt}}
     {targetBone {boneName=dyn_RB_jacket_dyn1_jnt}}
     {targetBone {boneName=dyn_RB_jacket_dyn2_jnt}}
     {targetBone {boneName=dyn_B_jacket_dyn1_jnt}}
     {targetBone {boneName=dyn_B_jacket_dyn2_jnt}}
    ]
    constraintWeights [1,1,1,1,1,1,1,1,1,1,1,1,1,1,1,1,1,1,1,1,1,1,1]
   }
   n10 = "Constraint - R_upper_zipper_jnt (PPPPPPPPPPPPPPPPPPPPPPPPPPPPPPPPP)"!(
    ComponentPose=@n9
    Alpha=1
    ConstraintWeights_0=0
    ConstraintWeights_1:float=0.0002
    ConstraintWeights_2:float=0.0006
    ConstraintWeights_3:float=0.0016
    ConstraintWeights_4:float=0.4694
    ConstraintWeights_5:float=0.1534
    ConstraintWeights_6:float=0
    ConstraintWeights_7:float=0
    ConstraintWeights_8:float=0
    ConstraintWeights_9:float=0
    ConstraintWeights_10:float=0
    ConstraintWeights_11:float=0
    ConstraintWeights_12:float=0.124
    ConstraintWeights_13:float=0.0922
    ConstraintWeights_14:float=0.0198
    ConstraintWeights_15:float=0.0001
    ConstraintWeights_16:float=0.0876
    ConstraintWeights_17:float=0.0003
    ConstraintWeights_18:float=0.0002
    ConstraintWeights_19:float=0.0001
    ConstraintWeights_20:float=0
    ConstraintWeights_21:float=0.0005
    ConstraintWeights_22:float=0
    ConstraintWeights_23:float=0.0497
    ConstraintWeights_24:float=0.0001
    ConstraintWeights_25:float=0.0003
    ConstraintWeights_26:float=0
    ConstraintWeights_27:float=0
    ConstraintWeights_28:float=0
    ConstraintWeights_29:float=0
    ConstraintWeights_30:float=0
    ConstraintWeights_31:float=0
    ConstraintWeights_32:float=0
   ) -> (Pose) {
    alphaBoolBlend {blendOption=Linear}
    boneToModify {boneName=R_upper_zipper_jnt}
    constraintSetup [
     {targetBone {boneName=pelvis}}
     {targetBone {boneName=spine_01}}
     {targetBone {boneName=spine_02}}
     {targetBone {boneName=spine_03}}
     {targetBone {boneName=spine_04}}
     {targetBone {boneName=spine_05}}
     {targetBone {boneName=clavicle_l}}
     {targetBone {boneName=upperarm_l}}
     {targetBone {boneName=upperarm_twist_01_l}}
     {targetBone {boneName=deform_deltoid_fr_l}}
     {targetBone {boneName=upperarm_twist_02_l}}
     {targetBone {boneName=deform_pectoral_fr_l}}
     {targetBone {boneName=clavicle_r}}
     {targetBone {boneName=upperarm_twist_01_r}}
     {targetBone {boneName=deform_deltoid_fr_r}}
     {targetBone {boneName=upperarm_twist_02_r}}
     {targetBone {boneName=deform_pectoral_fr_r}}
     {targetBone {boneName=neck_01}}
     {targetBone {boneName=neck_02}}
     {targetBone {boneName=head}}
     {targetBone {boneName=C_jaw}}
     {targetBone {boneName=dyn_R_collar_flap_jnt}}
     {targetBone {boneName=dyn_L_lapel_jnt}}
     {targetBone {boneName=dyn_R_lapel_jnt}}
     {targetBone {boneName=dyn_R_jacket_dyn1_jnt}}
     {targetBone {boneName=dyn_R_jacket_dyn2_jnt}}
     {targetBone {boneName=dyn_RB_jacket_dyn2_jnt}}
     {targetBone {boneName=dyn_LB_jacket_dyn1_jnt}}
     {targetBone {boneName=dyn_LB_jacket_dyn2_jnt}}
     {targetBone {boneName=dyn_B_jacket_dyn1_jnt}}
     {targetBone {boneName=dyn_B_jacket_dyn2_jnt}}
     {targetBone {boneName=dyn_L_jacket_dyn1_jnt}}
     {targetBone {boneName=dyn_L_jacket_dyn2_jnt}}
    ]
    constraintWeights [1,1,1,1,1,1,1,1,1,1,1,1,1,1,1,1,1,1,1,1,1,1,1,1,1,1,1,1,1,1,1,1,1]
   }
   n11 = "Constraint - L_hem_zipper_jnt (PPPPPPPPPP)"!(
    ComponentPose=@n10
    Alpha=1
    ConstraintWeights_0=0.0071
    ConstraintWeights_1:float=0.057
    ConstraintWeights_2:float=0
    ConstraintWeights_3:float=0
    ConstraintWeights_4:float=0
    ConstraintWeights_5:float=0
    ConstraintWeights_6:float=0
    ConstraintWeights_7:float=0.0111
    ConstraintWeights_8:float=0.0549
    ConstraintWeights_9:float=0.87
   ) -> (Pose) {
    alphaBoolBlend {blendOption=Linear}
    boneToModify {boneName=L_hem_zipper_jnt}
    constraintSetup [
     {targetBone {boneName=spine_02}}
     {targetBone {boneName=spine_04}}
     {targetBone {boneName=deform_deltoid_fr_l}}
     {targetBone {boneName=deform_pectoral_fr_l}}
     {targetBone {boneName=upperarm_twist_01_r}}
     {targetBone {boneName=C_jaw}}
     {targetBone {boneName=dyn_L_collar_flap_jnt}}
     {targetBone {boneName=dyn_LB_jacket_dyn2_jnt}}
     {targetBone {boneName=dyn_L_jacket_dyn2_jnt}}
     {targetBone {boneName=L_hem_zipper_jnt}}
    ]
    constraintWeights [1,1,1,1,1,1,1,1,1,1]
   }
   n12 = "Constraint - L_lower_zipper_jnt (PPPPPPPPPPPPPPPPPPPPPPPPP)"!(
    ComponentPose=@n11
    Alpha=1
    ConstraintWeights_0=0.0029
    ConstraintWeights_1:float=0.0064
    ConstraintWeights_2:float=0.0208
    ConstraintWeights_3:float=0.0452
    ConstraintWeights_4:float=0.6968
    ConstraintWeights_5:float=0.0705
    ConstraintWeights_6:float=0.0691
    ConstraintWeights_7:float=0.0512
    ConstraintWeights_8:float=0.0029
    ConstraintWeights_9:float=0
    ConstraintWeights_10:float=0.0046
    ConstraintWeights_11:float=0.0002
    ConstraintWeights_12:float=0
    ConstraintWeights_13:float=0
    ConstraintWeights_14:float=0
    ConstraintWeights_15:float=0
    ConstraintWeights_16:float=0.001
    ConstraintWeights_17:float=0
    ConstraintWeights_18:float=0.0001
    ConstraintWeights_19:float=0.008
    ConstraintWeights_20:float=0.0001
    ConstraintWeights_21:float=0.0026
    ConstraintWeights_22:float=0.0037
    ConstraintWeights_23:float=0.0111
    ConstraintWeights_24:float=0.0029
   ) -> (Pose) {
    alphaBoolBlend {blendOption=Linear}
    boneToModify {boneName=L_lower_zipper_jnt}
    constraintSetup [
     {targetBone {boneName=pelvis}}
     {targetBone {boneName=spine_01}}
     {targetBone {boneName=spine_02}}
     {targetBone {boneName=spine_03}}
     {targetBone {boneName=spine_04}}
     {targetBone {boneName=spine_05}}
     {targetBone {boneName=clavicle_l}}
     {targetBone {boneName=upperarm_twist_01_l}}
     {targetBone {boneName=deform_deltoid_fr_l}}
     {targetBone {boneName=upperarm_twist_02_l}}
     {targetBone {boneName=deform_pectoral_fr_l}}
     {targetBone {boneName=clavicle_r}}
     {targetBone {boneName=neck_01}}
     {targetBone {boneName=neck_02}}
     {targetBone {boneName=head}}
     {targetBone {boneName=dyn_L_collar_flap_jnt}}
     {targetBone {boneName=dyn_L_lapel_jnt}}
     {targetBone {boneName=dyn_RB_jacket_dyn2_jnt}}
     {targetBone {boneName=dyn_LB_jacket_dyn1_jnt}}
     {targetBone {boneName=dyn_LB_jacket_dyn2_jnt}}
     {targetBone {boneName=dyn_B_jacket_dyn1_jnt}}
     {targetBone {boneName=dyn_B_jacket_dyn2_jnt}}
     {targetBone {boneName=dyn_L_jacket_dyn1_jnt}}
     {targetBone {boneName=dyn_L_jacket_dyn2_jnt}}
     {targetBone {boneName=L_hem_zipper_jnt}}
    ]
    constraintWeights [1,1,1,1,1,1,1,1,1,1,1,1,1,1,1,1,1,1,1,1,1,1,1,1,1]
   }
   Output Pose(Result=Component To Local(ComponentPose=@n12)) {layerGroup=None}
  ]}
 EventGraph {nodes [Event BlueprintUpdateAnimation()]}
]
variables []
