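# Control Rig for the F_MED_Lurk face accessory part that re-aims the back hair chain root as the head rotates.
# It sets the initial local offset of helper bone spine_05_AIM, then shifts it further by a clamped weight remapped from the head's local Euler Z rotation.
# An Aim node points back_chain_master_jnt's -Z at that helper, using neck_01 as the secondary target; the downstream dyn_ joints follow as children.
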ControlRigBlueprint /BRCosmetics/Characters/Player/Female/Medium/Bodies/F_MED_Lurk/Meshes/Parts/F_MED_Lurk_FaceAcc_CtrlRig
graphs [
 RigVMFunctionLibrary {links [], nodes []}
 RigVMModel {nodes [
   RigUnit_BeginExecution = &RigUnit_BeginExecution()
   &RigUnit_SetTransform(
    Item=(Type=Bone,Name="spine_05_AIM")
    Space=LocalSpace
    Value=(Rotation=(X=0,Y=0,Z=0,W=1),Translation=(X=0,Y=-9,Z=0),Scale3D=(X=1,Y=1,Z=1))
    Weight=1
    bInitial=true
    ExecutePin=@RigUnit_BeginExecution.ExecutePin
   )
   n1 = &RigUnit_GetTransform(Item=(Type=Bone,Name="head"), Space=LocalSpace)
   > &RigUnit_ModifyTransforms(
    ItemToModify=((Item=(Type=Bone,Name="spine_05_AIM"),Transform=(Rotation=(X=0,Y=0,Z=0,W=1),Translation=(X=0,Y=-3,Z=0),Scale3D=(X=1,Y=1,Z=1))))
    Weight=&RigVMFunction_MathFloatRemap(
     SourceMaximum=-35
     SourceMinimum=-10
     bClamp=true
     Value=&RigVMFunction_MathQuaternionToEuler(RotationOrder=ZYX, Value=@n1.Transform.Rotation).Result.Z
    ).Result
   )
   > /Script/ControlRig.RigUnit_AimItem(
    DebugSettings=(Scale=10,WorldOffset=(Rotation=(X=0,Y=0,Z=0,W=1),Translation=(X=0,Y=0,Z=0),Scale3D=(X=1,Y=1,Z=1)))
    Item=(Type=Bone,Name="back_chain_master_jnt")
    Primary=(Weight=1,Axis=(X=0,Y=0,Z=-1),Target=(X=1,Y=0,Z=0),Kind=Location,Space=(Type=Bone))
    Secondary=(Weight=1,Axis=(X=0,Y=1,Z=0),Target=(X=1,Y=0,Z=0),Kind=Location,Space=(Type=Bone))
    Weight=1
    Primary.Target=&RigUnit_GetTransform(Item=(Type=Bone,Name="spine_05_AIM")).Transform.Translation
    Secondary.Target=&RigUnit_GetTransform(Item=(Type=Bone,Name="neck_01")).Transform.Translation
   ) {is=struct RigVMUnitNode}
  ]}
]
hierarchy [
 root {loc=<0 0 0>, rot=<0 0 0 1>}
 pelvis {parent=root, loc=<0 -0.5301 98.0571>, rot=<0.0127 0.707 -0.0127 -0.707>}
 spine_01 {parent=pelvis, loc=<6.1208 -0.9328 0>, rot=<0 0 0.034 -0.9994>}
 spine_02 {parent=spine_01, loc=<6.5996 -0.2095 0>, rot=<0 0 0.0238 -0.9997>}
 spine_03 {parent=spine_02, loc=<6.4914 -0.1237 0>, rot=<0 0 0.0149 -0.9999>}
 spine_04 {parent=spine_03, loc=<6.9598 0.0794 0>, rot=<0 0 -0.0132 -0.9999>}
 spine_05 {parent=spine_04, loc=<6.9987 0 0>, rot=<0 0 -0.0015 -1>}
 clavicle_l {parent=spine_05, loc=<5.4831 -0.1053 -3.2946>, rot=<0.0465 -0.7289 -0.009 -0.6829>}
 upperarm_l {parent=clavicle_l, loc=<9.8658 -1.2136 -0.4105>, rot=<0.0208 -0.486 -0.0242 -0.8734>}
 clavicle_r {parent=spine_05, loc=<5.4831 -0.1053 3.2946>, rot=<0.7289 0.0465 -0.6829 0.009>}
 upperarm_r {parent=clavicle_r, loc=<-9.8658 1.2136 0.4104>, rot=<0.0208 -0.486 -0.0242 -0.8734>}
 upperarm_twist_01_r {parent=upperarm_r, loc=<-7.2135 0.1134 0>, rot=<0 0 0 1>}
 neck_01 {parent=spine_05, loc=<11.3138 0.1529 0>, rot=<0 0 -0.2003 -0.9797>}
 neck_02 {parent=neck_01, loc=<4.5627 0 0>, rot=<0 0 0.0504 -0.9987>}
 head {parent=neck_02, loc=<5.3914 0 0>, rot=<0 0 0.115 -0.9934>}
 hair_master_jnt {parent=head, loc=<14.0436 -0.4312 0.0934>, rot=<0.0029 -0.7071 -0.0029 -0.7071>}
 back_chain_master_jnt {parent=hair_master_jnt, loc=<-0.9162 -8.0687 -6.353>, rot=<0 0 0 1>}
 dyn_back_chain_1_jnt {parent=back_chain_master_jnt, loc=<0.0072 -0.3544 -8.664>, rot=<0 -0.7125 0 -0.7017>}
 dyn_back_chain_2_jnt {parent=dyn_back_chain_1_jnt, loc=<6.1567 -2.2131 0.0765>, rot=<0 0.0099 0 -1>}
 dyn_back_chain_3_jnt {parent=dyn_back_chain_2_jnt, loc=<7.233 -1.3622 0>, rot=<0 0 0 1>}
 R_chain_master {parent=hair_master_jnt, loc=<-8.2745 -0.6223 -7.4988>, rot=<0 0 0 1>}
 dyn_R_chain_1_jnt {parent=R_chain_master, loc=<0.3618 -1.2345 -6.5932>, rot=<0.1871 -0.5452 0.4213 -0.7002>}
 dyn_R_chain_2_jnt {parent=dyn_R_chain_1_jnt, loc=<13.2262 1.4789 -0.9834>, rot=<-0.0707 -0.2575 -0.427 -0.8639>}
 bangs_master_jnt {parent=hair_master_jnt, loc=<0.787 6.4295 -1.2351>, rot=<0 0 0 1>}
 dyn_fringe_upper_jnt {parent=bangs_master_jnt, loc=<-3.0169 3.2824 -4.1975>, rot=<0 -0.9181 0 -0.3964>}
 dyn_fringe_lower_jnt {parent=dyn_fringe_upper_jnt, loc=<7.6316 -4.9635 -0.1468>, rot=<0 0.2679 0 -0.9634>}
 L_chain_master_jnt {parent=hair_master_jnt, loc=<4.9876 -0.5005 -5.1649>, rot=<0 0 0 1>}
 dyn_L_chain_1_jnt {parent=L_chain_master_jnt, loc=<1.4666 -3.2292 -3.7016>, rot=<0.2098 -0.6648 0.0764 -0.7129>}
 dyn_L_chain_2_jnt {parent=dyn_L_chain_1_jnt, loc=<11.1352 1.768 1.3964>, rot=<-0.1448 -0.1019 0.0085 -0.9842>}
 dyn_L_chain_flip_jnt {parent=dyn_L_chain_2_jnt, loc=<3.5106 1.5059 3.3074>, rot=<0.4986 -0.7558 -0.1478 0.3979>}
 dyn_L_chain_3_jnt {parent=dyn_L_chain_2_jnt, loc=<10.5055 -1.408 -3.4526>, rot=<-0.014 0.2801 -0.1306 -0.9509>}
 dyn_L_behindear_jnt {parent=dyn_L_chain_1_jnt, loc=<4.3268 1.6098 3.07>, rot=<-0.5429 0.4434 0.3989 -0.5912>}
 spine_05_AIM {parent=spine_05, loc=<0 0 0>, rot=<0 0 0 1>}
]
