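# Physics asset for the Lurk female medium body's FaceAcc mesh part, providing capsule bodies for dynamic hair chains, fringe and a left shoulder piece.
# Left, right, back and bangs chains branch from hair_master_jnt on the head, with limited swing and twist between segments.
# Free-angle constraints cross-link both side chains to the back chain.
# Head, neck, clavicle and upper-spine capsules form the body volume, with free-angle constraints to spine_04.
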PhysicsAsset /BRCosmetics/Characters/Player/Female/Medium/Bodies/F_MED_Lurk/Meshes/Parts/F_MED_Lurk_FaceAcc_Physics
bodies {
 L_chain_master_jnt [{shapeType=Capsule, length=0.5, radius=0.5, shapeName=L_chain_master_jnt_capsule}]
 R_chain_master [{shapeType=Capsule, length=0.5, radius=0.5, shapeName=R_chain_master_capsule}]
 back_chain_master_jnt [{shapeType=Capsule, length=0.5, radius=0.5, shapeName=back_chain_master_jnt_capsule, rotation=<rot 0 83.0181 0>}]
 bangs_master_jnt [{shapeType=Capsule, length=0.5, radius=0.5, shapeName=bangs_master_jnt_capsule}]
 clavicle_l [{
   shapeType=Capsule
   length=10.4045
   radius=7.2811
   shapeName=head_capsule
   center=<8.826 -1.3703 -0.6439>
   rotation=<rot -90 179.9963 -179.9963>
  }]
 clavicle_r [{shapeType=Capsule, length=10.4045, radius=7.2811, shapeName=head_capsule, center=<-9.3907 -2.6843 2.1065>, rotation=<rot -90 -1.8984 1.8984>}]
 dyn_L_behindear_jnt [{
   shapeType=Capsule
   length=1.5
   radius=1.25
   shapeName=dyn_L_behindear_jnt_capsule
   center=<1.8556 0.2297 0>
   rotation=<rot -0.0001 97.0557 90.0001>
  }]
 dyn_L_chain_1_jnt [{
   shapeType=Capsule
   length=4
   radius=3.25
   shapeName=dyn_L_chain_1_jnt_capsule
   center=<5.9624 0.0558 0.6872>
   rotation=<rot -0.0001 99.4065 100.0006>
  }]
 dyn_L_chain_2_jnt [{
   shapeType=Capsule
   length=4
   radius=2.5
   shapeName=dyn_L_chain_2_jnt_capsule
   center=<3.8144 -1.5374 -0.2122>
   rotation=<rot 59.9997 -0.0001 10.0011>
  }]
 dyn_L_chain_3_jnt [{shapeType=Capsule, length=3, radius=2, shapeName=dyn_L_chain_3_jnt_capsule, center=<2.4 -0.8926 1.365>, rotation=<rot 0 -89.9999 50.0003>}]
 dyn_L_chain_flip_jnt [{
   shapeType=Capsule
   length=2.5
   radius=2
   shapeName=dyn_L_chain_flip_jnt_capsule
   center=<0.9986 0.0001 0.176>
   rotation=<rot 80 -179.9998 -179.9998>
  }]
 dyn_L_shoulder_piece [{
   shapeType=Capsule
   length=2.6063
   radius=1.4715
   shapeName=dyn_L_shoulder_piece_capsule
   center=<1.0603 2.5591 -0.187>
   rotation=<rot -79.9999 179.9995 -129.9994>
  }]
 dyn_R_chain_1_jnt [{
   shapeType=Capsule
   length=6.75
   radius=3.75
   shapeName=dyn_R_chain_1_jnt_capsule
   center=<7.2401 1.5623 -1.3918>
   rotation=<rot 88.2659 -85.07 -94.9197>
  }]
 dyn_R_chain_2_jnt [{
   shapeType=Capsule
   length=4.23
   radius=2.0293
   shapeName=dyn_R_chain_2_jnt_capsule
   center=<3.3626 0.0994 -0.5307>
   rotation=<rot -21.0143 -131.4689 -41.1778>
  }]
 dyn_R_chain_flip_jnt [{
   shapeType=Capsule
   length=4.23
   radius=2.0293
   shapeName=dyn_R_chain_2_jnt_capsule
   center=<3.3626 0.0994 -0.5307>
   rotation=<rot -21.0143 -131.4689 -41.1778>
  }]
 dyn_back_chain_1_jnt [{shapeType=Capsule, length=4.2743, radius=4.7318, shapeName=dyn_back_chain_1_jnt_capsule, center=<4.8647 -2.9941 0.9886>}]
 dyn_back_chain_2_jnt [{shapeType=Capsule, length=7.3801, radius=3.9125, shapeName=dyn_back_chain_2_jnt_capsule, center=<5.9323 -1.7837 -0.0001>}]
 dyn_back_chain_3_jnt [{
   shapeType=Capsule
   length=5.2737
   radius=4.3434
   shapeName=dyn_back_chain_3_jnt_capsule
   center=<6.4612 -1.5678 0>
   rotation=<rot 0.0005 0.0001 0>
  }]
 dyn_fringe_lower_jnt [{
   shapeType=Capsule
   length=5
   radius=1.75
   shapeName=dyn_fringe_lower_jnt_capsule
   center=<4.0663 -0.7026 1.4798>
   rotation=<rot 0 -89.9999 69.9999>
  }]
 dyn_fringe_upper_jnt [{
   shapeType=Capsule
   length=3.25
   radius=3
   shapeName=dyn_fringe_upper_jnt_capsule
   center=<5.2854 -0.667 0.2473>
   rotation=<rot -9.2614 -119.7005 -85.811>
  }]
 hair_master_jnt [{shapeType=Capsule, length=0.5, radius=0.5, shapeName=hair_master_jnt_capsule, rotation=<rot 0 110.2366 0>}]
 head [{shapeType=Capsule, length=5, radius=7, shapeName=head_capsule, center=<0 1.8865 0>, rotation=<rot -90 -1.2659 1.2661>}]
 neck_01 [{shapeType=Capsule, length=3, radius=3, shapeName=neck_01_capsule}]
 neck_02 [{shapeType=Capsule, length=3.75, radius=4, shapeName=neck_02_capsule, center=<1.4124 0 0>, rotation=<rot -90 0 0>}]
 spine_04 [{shapeType=Capsule, length=5, radius=10, shapeName=spine_04_capsule}]
 spine_05 [
  {shapeType=Capsule, length=16, radius=9.75, shapeName=spine_05_capsule, center=<1.6525 0.5237 -0.0001>, rotation=<rot 0 -89.9999 0>}
  {shapeType=Capsule, length=61.75, radius=8.75, shapeName=spine_05_capsule, center=<5.9775 5.7152 0>}
 ]
}
constraints [
 {
  bone1Name=back_chain_master_jnt
  bone2Name=hair_master_jnt
  swing1LimitDegrees=45
  swing1Motion=Free
  swing2LimitDegrees=45
  swing2Motion=Free
  twistLimitDegrees=45
  twistMotion=Free
 }
 {bone1Name=dyn_back_chain_1_jnt, bone2Name=back_chain_master_jnt, swing1LimitDegrees=60, swing2LimitDegrees=60, twistLimitDegrees=0}
 {bone1Name=dyn_back_chain_2_jnt, bone2Name=dyn_back_chain_1_jnt, swing1LimitDegrees=60, swing2LimitDegrees=60, twistLimitDegrees=0}
 {bone1Name=dyn_back_chain_3_jnt, bone2Name=dyn_back_chain_2_jnt, swing1LimitDegrees=60, swing2LimitDegrees=60, twistLimitDegrees=0}
 {
  bone1Name=R_chain_master
  bone2Name=hair_master_jnt
  swing1LimitDegrees=45
  swing1Motion=Free
  swing2LimitDegrees=45
  swing2Motion=Free
  twistLimitDegrees=45
  twistMotion=Free
 }
 {bone1Name=dyn_R_chain_1_jnt, bone2Name=R_chain_master, swing1LimitDegrees=25, swing2LimitDegrees=25, twistLimitDegrees=2}
 {bone1Name=dyn_R_chain_2_jnt, bone2Name=dyn_R_chain_1_jnt, swing1LimitDegrees=5, swing2LimitDegrees=5, twistLimitDegrees=0}
 {
  bone1Name=bangs_master_jnt
  bone2Name=hair_master_jnt
  swing1LimitDegrees=45
  swing1Motion=Free
  swing2LimitDegrees=45
  swing2Motion=Free
  twistLimitDegrees=45
  twistMotion=Free
 }
 {bone1Name=dyn_fringe_upper_jnt, bone2Name=bangs_master_jnt, swing1LimitDegrees=25, swing2LimitDegrees=25, twistLimitDegrees=2}
 {bone1Name=dyn_fringe_lower_jnt, bone2Name=dyn_fringe_upper_jnt, swing1LimitDegrees=25, swing2LimitDegrees=25, twistLimitDegrees=2}
 {
  bone1Name=L_chain_master_jnt
  bone2Name=hair_master_jnt
  swing1LimitDegrees=45
  swing1Motion=Free
  swing2LimitDegrees=45
  swing2Motion=Free
  twistLimitDegrees=45
  twistMotion=Free
 }
 {bone1Name=dyn_L_chain_1_jnt, bone2Name=L_chain_master_jnt, swing1LimitDegrees=15, swing2LimitDegrees=15, twistLimitDegrees=2}
 {bone1Name=dyn_L_chain_3_jnt, bone2Name=dyn_L_chain_2_jnt, swing1LimitDegrees=5, swing2LimitDegrees=5, twistLimitDegrees=2}
 {bone1Name=dyn_L_behindear_jnt, bone2Name=dyn_L_chain_1_jnt, swing1LimitDegrees=15, swing2LimitDegrees=15, twistLimitDegrees=2}
 {
  bone1Name=hair_master_jnt
  bone2Name=head
  swing1LimitDegrees=45
  swing1Motion=Free
  swing2LimitDegrees=45
  swing2Motion=Free
  twistLimitDegrees=45
  twistMotion=Free
 }
 {bone1Name=dyn_L_chain_flip_jnt, bone2Name=dyn_L_chain_2_jnt, swing1LimitDegrees=15, swing2LimitDegrees=15, twistLimitDegrees=2}
 {bone1Name=dyn_L_chain_2_jnt, bone2Name=dyn_L_chain_1_jnt, swing1LimitDegrees=10, swing2LimitDegrees=10, twistLimitDegrees=2}
 {
  bone1Name=spine_05
  bone2Name=spine_04
  swing1LimitDegrees=45
  swing1Motion=Free
  swing2LimitDegrees=45
  swing2Motion=Free
  twistLimitDegrees=45
  twistMotion=Free
 }
 {
  bone1Name=clavicle_l
  bone2Name=spine_04
  swing1LimitDegrees=45
  swing1Motion=Free
  swing2LimitDegrees=45
  swing2Motion=Free
  twistLimitDegrees=45
  twistMotion=Free
 }
 {
  bone1Name=clavicle_r
  bone2Name=spine_04
  swing1LimitDegrees=45
  swing1Motion=Free
  swing2LimitDegrees=45
  swing2Motion=Free
  twistLimitDegrees=45
  twistMotion=Free
 }
 {
  bone1Name=neck_01
  bone2Name=spine_04
  swing1LimitDegrees=45
  swing1Motion=Free
  swing2LimitDegrees=45
  swing2Motion=Free
  twistLimitDegrees=45
  twistMotion=Free
 }
 {
  bone1Name=neck_02
  bone2Name=spine_04
  swing1LimitDegrees=45
  swing1Motion=Free
  swing2LimitDegrees=45
  swing2Motion=Free
  twistLimitDegrees=45
  twistMotion=Free
 }
 {
  bone1Name=head
  bone2Name=spine_04
  swing1LimitDegrees=45
  swing1Motion=Free
  swing2LimitDegrees=45
  swing2Motion=Free
  twistLimitDegrees=45
  twistMotion=Free
 }
 {bone1Name=dyn_R_chain_flip_jnt, bone2Name=dyn_R_chain_2_jnt, swing1LimitDegrees=5, swing2LimitDegrees=5, twistLimitDegrees=0}
 {
  bone1Name=dyn_R_chain_2_jnt
  bone2Name=dyn_back_chain_2_jnt
  swing1LimitDegrees=45
  swing1Motion=Free
  swing2LimitDegrees=45
  swing2Motion=Free
  twistLimitDegrees=45
  twistMotion=Free
 }
 {
  bone1Name=dyn_L_chain_3_jnt
  bone2Name=dyn_back_chain_2_jnt
  swing1LimitDegrees=45
  swing1Motion=Free
  swing2LimitDegrees=45
  swing2Motion=Free
  twistLimitDegrees=45
  twistMotion=Free
 }
 {bone1Name=dyn_L_shoulder_piece, bone2Name=dyn_L_chain_2_jnt, swing1LimitDegrees=15, swing2LimitDegrees=15, twistLimitDegrees=0}
]
