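# Animation Blueprint driving the chain face accessory on the female medium Lurk cosmetic body.
# Copies the attached parent's pose through a Control Rig, offsets the right chain master, then simulates RigidBody physics in head space.
# Post-physics additive bone corrections follow head-back rotation against spine_05, plus front-end-only offsets.
# Targeting bone tweaks sit disconnected from output, headTilt is unused, and targetUp's multiplier is unwired.
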AnimBlueprint /BRCosmetics/Characters/Player/Female/Medium/Bodies/F_MED_Lurk/Meshes/Parts/F_MED_Lurk_FaceAcc_AnimBP
target_skeleton=&/F_MED_Lurk_FaceAcc_Skeleton
graphs [
 AnimGraph {nodes [
   n1 = Get targetUp() -> (targetUp:double)
   "Transform (Modify) Bone - Bone: dyn_R_chain_1_jnt"(Alpha=@n1, Translation="0.142,-1.631,0.000", Rotation="0.000,15.001,0.000", Scale="1,1,1") {
    rotationMode=BMM_Additive
    rotationSpace=BCS_BoneSpace
    translationMode=BMM_Additive
    translationSpace=BCS_BoneSpace
    alphaBoolBlend {blendOption=Linear}
    alphaScaleBiasClamp {bInterpResult=true, interpSpeedDecreasing=2, interpSpeedIncreasing=6}
    boneToModify {boneName=dyn_R_chain_1_jnt}
    rotation {yaw=15.0015}
    translation=<0.142 -1.631>
   }
   "Transform (Modify) Bone - Bone: R_chain_master"(Alpha=@n1, Translation="0.090,0.829,5.476", Rotation="-0.252,-14.512,-20.603", Scale="1,1,1") {
    rotationMode=BMM_Additive
    rotationSpace=BCS_BoneSpace
    translationMode=BMM_Additive
    translationSpace=BCS_BoneSpace
    alphaBoolBlend {blendOption=Linear}
    alphaScaleBiasClamp {bInterpResult=true, interpSpeedDecreasing=2, interpSpeedIncreasing=6}
    boneToModify {boneName=R_chain_master}
    rotation=<rot -0.252 -14.5124 -20.6028>
    translation=<0.09 0.829 5.476>
   }
   n2 = "To Float (Boolean)"(InBool=Get bIsTargeting())
   "Transform (Modify) Bone - Bone: dyn_R_chain_1_jnt"(Alpha=@n2, Translation="1,-5,-5", Rotation="0,0,-20", Scale="1,1,1") {
    rotationMode=BMM_Additive
    rotationSpace=BCS_BoneSpace
    translationMode=BMM_Additive
    translationSpace=BCS_BoneSpace
    alphaBoolBlend {blendOption=Linear}
    alphaScaleBiasClamp {bInterpResult=true, interpSpeedDecreasing=2, interpSpeedIncreasing=6}
    boneToModify {boneName=dyn_R_chain_1_jnt}
    rotation {roll=-20}
    translation=<1 -5 -5>
   }
   n3 = "Transform (Modify) Bone - Bone: dyn_R_chain_2_jnt"(Alpha=@n2, Translation="5,5,-1", Rotation="-5,-30,0", Scale="1,1,1") {
    rotationMode=BMM_Additive
    rotationSpace=BCS_BoneSpace
    translationMode=BMM_Additive
    translationSpace=BCS_BoneSpace
    alphaBoolBlend {blendOption=Linear}
    alphaScaleBiasClamp {bInterpResult=true, interpSpeedDecreasing=2, interpSpeedIncreasing=6}
    boneToModify {boneName=dyn_R_chain_2_jnt}
    rotation {pitch=-5, yaw=-30}
    translation=<5 5 -1>
   }
   n4 = "Transform (Modify) Bone - Bone: dyn_R_chain_flip_jnt"(ComponentPose=@n3, Alpha=@n2, Translation="-3,0,-2", Rotation="0,0,0", Scale="1,1,1") {
    rotationMode=BMM_Additive
    rotationSpace=BCS_BoneSpace
    translationMode=BMM_Additive
    translationSpace=BCS_BoneSpace
    alphaBoolBlend {blendOption=Linear}
    alphaScaleBiasClamp {bInterpResult=true, interpSpeedDecreasing=2, interpSpeedIncreasing=6}
    boneToModify {boneName=dyn_R_chain_flip_jnt}
    translation {x=-3, z=-2}
   }
   n5 = "Transform (Modify) Bone - Bone: dyn_L_chain_2_jnt"(ComponentPose=@n4, Alpha=1, Translation="0,-1.5,0", Rotation="0,0,0", Scale="1,1,1") {
    translationMode=BMM_Additive
    translationSpace=BCS_BoneSpace
    alphaBoolBlend {blendOption=Linear}
    boneToModify {boneName=dyn_L_chain_2_jnt}
    translation {y=-1.5}
   }
   "Transform (Modify) Bone - Bone: R_chain_master"(ComponentPose=@n5, Alpha=@n2, Translation="0,0,5", Rotation="0,0,0", Scale="1,1,1") {
    rotationMode=BMM_Additive
    rotationSpace=BCS_BoneSpace
    translationMode=BMM_Additive
    translationSpace=BCS_BoneSpace
    alphaBoolBlend {blendOption=Linear}
    alphaScaleBiasClamp {bInterpResult=true, interpSpeedDecreasing=2, interpSpeedIncreasing=6}
    boneToModify {boneName=R_chain_master}
    translation {z=5}
   }
   "Transform (Modify) Bone - Bone: L_chain_master_jnt"(Alpha=1, Translation="0,3,2", Rotation="0,0,0", Scale="1,1,1") {
    rotationMode=BMM_Additive
    rotationSpace=BCS_BoneSpace
    translationMode=BMM_Additive
    translationSpace=BCS_BoneSpace
    alphaBoolBlend {blendOption=Linear}
    alphaScaleBiasClamp {bInterpResult=true, interpSpeedDecreasing=6, interpSpeedIncreasing=6}
    boneToModify {boneName=L_chain_master_jnt}
    translation {y=3, z=2}
   }
   n6 = Control Rig(Source=Copy Pose From Mesh() {bUseAttachedParent=true}) {
    alphaBoolBlend {blendOption=Linear}
    controlRigAssetReference {blueprintRigClass=&/F_MED_Lurk_FaceAcc_CtrlRig.F_MED_Lurk_FaceAcc_CtrlRig_C}
   }
   n7 = "Transform (Modify) Bone - Bone: R_chain_master"(ComponentPose=Local To Component(LocalPose=@n6), Alpha=@n1, Translation="3,-7.000,7.000", Rotation="-5.000,0.000,0.000", Scale="1,1,1") {
    rotationMode=BMM_Additive
    rotationSpace=BCS_BoneSpace
    translationMode=BMM_Additive
    translationSpace=BCS_BoneSpace
    alphaBoolBlend {blendOption=Linear}
    alphaScaleBiasClamp {bInterpResult=true, interpSpeedDecreasing=6, interpSpeedIncreasing=6}
    boneToModify {boneName=R_chain_master}
    rotation {pitch=-5}
    translation=<3 -7 7>
   }
   n8 = RigidBody(ComponentPose=@n7) {
    bClampLinearTranslationLimitToRefPose=true
    bOverrideWorldGravity=true
    overridePhysicsAsset=&/F_MED_Lurk_FaceAcc_Physics
    simulationSpace=BaseBoneSpace
    alphaBoolBlend {blendOption=Linear}
    baseBoneRef {boneName=head}
    componentAppliedLinearAccClamp=<3000 3000 3000>
    componentLinearAccScale=<0.5 0.2 0.2>
    componentLinearVelScale=<0.5 0.2 0.2>
    overrideWorldGravity {z=-980}
   }
   n9 = "Transform (Modify) Bone - Bone: dyn_R_chain_1_jnt"(
    ComponentPose=@n8
    Alpha=Get headBack() -> (headBack:double)
    Translation="2.086,-4.703,1.013"
    Rotation="-10.799,18.931,9.473"
    Scale="1,1,1"
   ) {
    rotationMode=BMM_Additive
    rotationSpace=BCS_BoneSpace
    translationMode=BMM_Additive
    translationSpace=BCS_BoneSpace
    alphaBoolBlend {blendOption=Linear}
    alphaScaleBiasClamp {bInterpResult=true, interpSpeedDecreasing=6, interpSpeedIncreasing=6}
    boneToModify {boneName=dyn_R_chain_1_jnt}
    rotation=<rot -10.799 18.9311 9.4734>
    translation=<2.086 -4.703 1.013>
   }
   n10 = Get bIsInFrontEnd()
   n11 = "Transform (Modify) Bone - Bone: dyn_L_chain_1_jnt"!(ComponentPose=@n9, bAlphaBoolEnabled:bool=@n10, Translation="0,-0.5,0", Rotation="0,0,-10", Scale="1,1,1") -> (Pose) {
    alphaInputType=Bool
    rotationMode=BMM_Additive
    rotationSpace=BCS_BoneSpace
    translationMode=BMM_Additive
    translationSpace=BCS_BoneSpace
    alphaBoolBlend {blendOption=Linear}
    boneToModify {boneName=dyn_L_chain_1_jnt}
    rotation {roll=-10}
    translation {y=-0.5}
   }
   n12 = "Transform (Modify) Bone - Bone: R_chain_master"!(ComponentPose=@n11, bAlphaBoolEnabled:bool=@n10, Translation="0,-2,1", Rotation="0,0,0", Scale="1,1,1") -> (Pose) {
    alphaInputType=Bool
    rotationMode=BMM_Additive
    rotationSpace=BCS_BoneSpace
    translationMode=BMM_Additive
    translationSpace=BCS_BoneSpace
    alphaBoolBlend {blendOption=Linear}
    boneToModify {boneName=R_chain_master}
    translation {y=-2, z=1}
   }
   n13 = "Transform (Modify) Bone - Bone: dyn_R_chain_2_jnt"!(ComponentPose=@n12, bAlphaBoolEnabled:bool=@n10, Translation="0,3,0", Rotation="0,0,0", Scale="1,1,1") -> (Pose) {
    alphaInputType=Bool
    rotationMode=BMM_Additive
    rotationSpace=BCS_BoneSpace
    translationMode=BMM_Additive
    translationSpace=BCS_BoneSpace
    alphaBoolBlend {blendOption=Linear}
    boneToModify {boneName=dyn_R_chain_2_jnt}
    translation {y=3}
   }
   Output Pose(Result=Component To Local(ComponentPose=@n13)) {layerGroup=None}
  ]}
 EventGraph {nodes [
   Event BlueprintUpdateAnimation()
   n1 = GetOwnerSkeletalMeshComponent()
   > n2 = Set headBack(headBack:double=GetAlphaFromReferenceTransformDelta(SkeletalMeshComponent=@n1, BoneName=head, BaseName=spine_05, InputRange=(X=0,Y=-35), TwistAxis="1,0,0")) -> (Output_Get:double)
   > Set targetUp(targetUp:double=float * float(A=@n2) {is=title K2Node_PromotableOperator}) -> (Output_Get:double)
   > Set headTilt(headTilt:double=GetAlphaFromReferenceTransformDelta(SkeletalMeshComponent=@n1, BoneName=head, BaseName=spine_05, DeltaComponentType=EulerY, InputRange=(X=32,Y=41), TwistAxis="1,0,0")) -> (Output_Get:double)
   Get bIsTargeting()
  ]}
]
variables [
 headBack {type=number}
 headTilt {type=number}
 targetUp {type=number}
]
